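# Parent material for corrugated roller surfaces, likely doors, in the SilverPopcorn set.
# It builds UVs from world position, a scrolling Offset and transformed bounds extents (switchable direction), blending two masked planes by vertex normal, then scaling and rotating.
# These drive diffuse, normal and spec textures, with spec green shared by metallic, specular and roughness, feeding MF_FN_DestructibleDamage.
# Several unconnected nodes remain, including object-scale UV math and vector rotation.
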Material /Hera/Sets/SilverPopcorn/Materials/M_Roller_Corregated_Parent
graph {
 properties {bAutomaticallySetUsageInEditor=false, bHasPixelAnimation=true, bUseMaterialAttributes=true, usedShadingModels=Default Lit}
 nodes [
  n1 = TextureCoordinate()
  n2 = MaterialFunctionCall() {materialFunction=&ObjectScale}
  Add(A=@n1, B=Multiply(A=@n2.Scale X) {constB=-1})
  n3 = Constant() {r=2.3}
  Divide(A=@n2.Scale X, B=@n3)
  MaterialFunctionCall!(VectorToRotate, LookAtVector, RestingVector) {materialFunction=/Engine/Functions/Engine_MaterialFunctions01/Vectors/RotateVector}
  MaterialFunctionCall(Vector 3=Transform(None=Constant3Vector() {constant=<rgba 1 0 0 0>}) {transformSourceType=TRANSFORMSOURCE_Instance}) {materialFunction=/Engine/Functions/Engine_MaterialFunctions02/Utility/VectorLength}
  n4 = Add(A=WorldPosition().XYZ, B=ScalarParameter() {defaultValue=1, parameterName=Offset})
  n5 = TransformPosition(None=Bounds().Half Extents) {transformType=TRANSFORMPOSSOURCE_World}
  n6 = Divide(
   A=StaticSwitchParameter(True=Subtract(A=@n4, B=@n5), False=Add(A=@n4, B=@n5)) {defaultValue=true, parameterName=Reverse Movement}
   B=Multiply(A=ScalarParameter() {defaultValue=256, parameterName=DoorTextureScale}) {constB=-1}
  )
  n7 = Rotator(
   Coordinate=LinearInterpolate(
    A=StaticComponentMaskParameter(None=@n6) {defaultB=true, defaultR=true, parameterName=Masks}
    B=StaticComponentMaskParameter(None=@n6) {defaultB=true, defaultG=true, parameterName=Masks90}
    Alpha=VertexNormalWS()
   )
   Time=Multiply(A=ScalarParameter() {defaultValue=6.28, desc=2x Pi for sideways, parameterName=Rotate}) {constB=6.26}
  )
  n8 = TextureSampleParameter2D(UVs=@n7) {parameterName=Spec, samplerType=SAMPLERTYPE_Masks, texture=/Hera/Sets/SilverPopcorn/Textures/T_SilverPopcorn_StudioWall_F_A_S}
  n9 = TextureSampleParameter2D(UVs=@n7) {parameterName=Normal, samplerType=SAMPLERTYPE_Normal, texture=/Hera/Sets/SilverPopcorn/Textures/T_SilverPopcorn_StudioWall_F_A_N}
  n10 = SetMaterialAttributes!(
   MaterialAttributes
   Base Color=TextureSampleParameter2D(UVs=@n7) {parameterName=Diffuse, texture=/Hera/Sets/SilverPopcorn/Textures/T_SilverPopcorn_StudioWall_F_A_D}.RGB
   Metallic=@n8.G
   Specular=@n8.G
   Roughness=@n8.G
   Anisotropy
   Emissive Color
   Opacity
   Opacity Mask
   Normal=@n9.RGB
   Tangent
  ) {attributeSetTypes [
    #1
    #2
    #3
    #4
    #5
    #6
    #7
    #8
    #9
    #10
   ]}
  MaterialFunctionCall(Material_In=@n10) {materialFunction=&MF_FN_DestructibleDamage}
  Multiply()
  OneMinus(None=@n1)
  RotateAboutAxis()
  Subtract(
   A=Add(A=@n1, B=ScalarParameter() {defaultValue=1, parameterName=Offset}) {constB=0.5}
   B=AppendVector(A=Multiply(A=@n2.Scale X, B=@n3), B=Constant())
  )
  Transform(None=@n6) {transformSourceType=TRANSFORMSOURCE_World, transformType=TRANSFORM_Local}
  VertexInterpolator(VS=PreSkinnedNormal())
 ]
}
